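# Sound cue for TalonPane narrator voice lines in the TidalCrane base activities Queen Hunts set, named as the on-entry bark for a "Below25" condition.
# A Random root picks with equal weight among ten wave players, each holding a numbered OnEntry recording.
# It uses the narrator VO sound class, the shared base-activities attenuation asset and a TalonPane-specific concurrency asset.
# The cue holds no threshold logic itself.
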SoundCue /TidalCrane_BaseActivities/Assets/VO/Cues/QueenHunts/VO_TidalCrane_TalonPane_OnEntry_Below25_Cue
root=Random_0
nodes [
 {
  id=Random_0
  children [WavePlayer_0,WavePlayer_1,WavePlayer_2,WavePlayer_3,WavePlayer_4,WavePlayer_5,WavePlayer_6,WavePlayer_7,WavePlayer_8,WavePlayer_9]
  properties {weights [1,1,1,1,1,1,1,1,1,1]}
 }
 {id=WavePlayer_0, properties {soundWaveAssetPtr=/TidalCrane_BaseActivities/Assets/VO/Waves/TalonPane/VO_TidalCrane_TalonPane_OnEntry_010}}
 {id=WavePlayer_1, properties {soundWaveAssetPtr=/TidalCrane_BaseActivities/Assets/VO/Waves/TalonPane/VO_TidalCrane_TalonPane_OnEntry_020}}
 {id=WavePlayer_2, properties {soundWaveAssetPtr=/TidalCrane_BaseActivities/Assets/VO/Waves/TalonPane/VO_TidalCrane_TalonPane_OnEntry_030}}
 {id=WavePlayer_3, properties {soundWaveAssetPtr=/TidalCrane_BaseActivities/Assets/VO/Waves/TalonPane/VO_TidalCrane_TalonPane_OnEntry_040}}
 {id=WavePlayer_4, properties {soundWaveAssetPtr=/TidalCrane_BaseActivities/Assets/VO/Waves/TalonPane/VO_TidalCrane_TalonPane_OnEntry_050}}
 {id=WavePlayer_5, properties {soundWaveAssetPtr=/TidalCrane_BaseActivities/Assets/VO/Waves/TalonPane/VO_TidalCrane_TalonPane_OnEntry_060}}
 {id=WavePlayer_6, properties {soundWaveAssetPtr=/TidalCrane_BaseActivities/Assets/VO/Waves/TalonPane/VO_TidalCrane_TalonPane_OnEntry_070}}
 {id=WavePlayer_7, properties {soundWaveAssetPtr=/TidalCrane_BaseActivities/Assets/VO/Waves/TalonPane/VO_TidalCrane_TalonPane_OnEntry_080}}
 {id=WavePlayer_8, properties {soundWaveAssetPtr=/TidalCrane_BaseActivities/Assets/VO/Waves/TalonPane/VO_TidalCrane_TalonPane_OnEntry_090}}
 {id=WavePlayer_9, properties {soundWaveAssetPtr=/TidalCrane_BaseActivities/Assets/VO/Waves/TalonPane/VO_TidalCrane_TalonPane_OnEntry_100}}
]
properties {
 attenuationSettings=/TidalCrane_BaseActivities/Assets/VO/AS_Fort_BaseActivities
 maxDistance=10100
 soundClassObject=/Game/Sounds/FortSoundClassesAndModes/Fort_VO_Narrator
 volumeMultiplier=1
 concurrencySet [/TidalCrane_BaseActivities/Assets/VO/BaseActivites_VO_TalonPane_Conc]
}
